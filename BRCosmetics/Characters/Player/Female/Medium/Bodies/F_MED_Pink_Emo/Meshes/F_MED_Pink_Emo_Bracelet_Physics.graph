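# Physics asset for the bracelet mesh of the Pink Emo medium female player cosmetic body.
# It has two capsule bodies: a larger one on the Dyn_Master_Bracelet bone and a small one on the right lower-arm twist bone.
# A single constraint joins them with both swings and twist locked, so its angular limit values are inactive and the bracelet body cannot rotate relative to the forearm through that joint.
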PhysicsAsset /BRCosmetics/Characters/Player/Female/Medium/Bodies/F_MED_Pink_Emo/Meshes/F_MED_Pink_Emo_Bracelet_Physics
bodies {
 Dyn_Master_Bracelet [{shapeType=Capsule, length=10, radius=2.5, shapeName=Dyn_Master_Bracelet_capsule, center=<1 0 0>, rotation=<rot 90 0 0>}]
 lowerarm_twist_01_r [{shapeType=Capsule, length=1, radius=1, shapeName=lowerarm_twist_01_r_capsule, rotation=<rot 90 0 0>}]
}
constraints [{
  bone1Name=Dyn_Master_Bracelet
  bone2Name=lowerarm_twist_01_r
  swing1LimitDegrees=20
  swing1Motion=Locked
  swing2LimitDegrees=5
  swing2Motion=Locked
  twistLimitDegrees=1
  twistMotion=Locked
 }]
