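# Crouch-named physics asset for the F_MED_Pink_Emo female player body, simulating a dangling belt against simple capsule bodies.
# Front and back belt bone chains hang from master anchors on the pelvis and join at their ends, closing a loop.
# Belt joints use tight angular limits, with the front 01-02 and 02-03 joints fully locked.
# Pelvis, lower spine and thigh capsules supply body volume, linked by free angular constraints.
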PhysicsAsset /BRCosmetics/Characters/Player/Female/Medium/Bodies/F_MED_Pink_Emo/Meshes/F_MED_Pink_Emo_Physics_CROUCH
bodies {
 Dyn_Belt_BK_02 [
  {shapeType=Capsule, length=10, radius=1.5, shapeName=Dyn_Belt_BK_02_capsule, center=<2.9221 0.0003 0.7562>, rotation=<rot 79.9996 0 0>}
  {
   shapeType=Capsule
   length=6
   radius=1
   shapeName=Dyn_Belt_BK_02_capsule
   center=<-4.5551 -0.0004 -0.43>
   rotation=<rot 64.9919 -180 -179.9999>
  }
 ]
 Dyn_Belt_BK_03 [{shapeType=Capsule, length=6, radius=2, shapeName=Dyn_Belt_BK_03_capsule, center=<2.5485 0.1899 0.8976>, rotation=<rot 79.9996 0 0>}]
 Dyn_Belt_BK_04 [{shapeType=Capsule, length=10, radius=2, shapeName=Dyn_Belt_BK_04_capsule, center=<3.0174 0.0001 0.0652>, rotation=<rot 75 0 0.1996>}]
 Dyn_Belt_BK_05 [{shapeType=Capsule, length=6, radius=2, shapeName=Dyn_Belt_BK_05_capsule, center=<1.8227 0.0001 0.5904>, rotation=<rot 30 0 0.1995>}]
 Dyn_Belt_FT_01 [{shapeType=Capsule, length=6, radius=1.5, shapeName=Dyn_Belt_FT_01_capsule, center=<2 0 0>, rotation=<rot 90 0 0>}]
 Dyn_Belt_FT_02 [{shapeType=Capsule, length=5.3347, radius=2.1653, shapeName=Dyn_Belt_FT_02_capsule, center=<1 0.0001 0.8226>, rotation=<rot 90 0 0>}]
 Dyn_Belt_FT_03 [{shapeType=Capsule, length=5.2299, radius=2.2701, shapeName=Dyn_Belt_FT_03_capsule, center=<0.9894 0 0.4268>, rotation=<rot 84.9923 180 180>}]
 Dyn_Belt_FT_04 [{shapeType=Capsule, length=10, radius=2, shapeName=Dyn_Belt_FT_04_capsule, center=<1 0 0.5686>, rotation=<rot 90 -102.0168 -101.8172>}]
 Dyn_Belt_FT_05 [
  {shapeType=Capsule, length=10, radius=2, shapeName=Dyn_Belt_FT_05_capsule, center=<1 0 0.6872>, rotation=<rot 90 0 0>}
  {shapeType=Capsule, length=10, radius=2, shapeName=Dyn_Belt_FT_05_capsule, center=<4.4475 0.0001 0.2226>, rotation=<rot 69.9994 0 0.1995>}
 ]
 Dyn_Belt_FT_06 [{shapeType=Capsule, length=6, radius=2.3, shapeName=Dyn_Belt_FT_06_capsule, center=<1.0452 0 1.9743>, rotation=<rot 74.9897 180 -179.8005>}]
 Dyn_Master_Belt_BK [{shapeType=Capsule, length=0.5, radius=0.5, shapeName=Dyn_Master_Belt_BK_capsule, rotation=<rot 0 -90 0>}]
 Dyn_Master_Belt_FT [{shapeType=Capsule, length=0.5, radius=0.5, shapeName=Dyn_Master_Belt_FT_capsule, rotation=<rot 0 -90 0>}]
 pelvis [
  {shapeType=Capsule, length=10, radius=9.2, shapeName=pelvis_capsule, center=<-2.8519 -1.2587 -0.2242>}
  {shapeType=Capsule, length=10, radius=10, shapeName=pelvis_capsule, center=<-5.2913 -1.8386 -0.0001>}
 ]
 spine_01 [{shapeType=Capsule, length=14, radius=10, shapeName=spine_01_capsule, center=<-0.0944 -0.3846 0>, rotation=<rot 90 0 0>}]
 thigh_l [
  {shapeType=Capsule, length=7, radius=7.5, shapeName=thigh_l_capsule, center=<-9.7008 -2.2229 3.2043>, rotation=<rot 90 0 15.0028>}
  {shapeType=Capsule, length=20, radius=7, shapeName=thigh_l_capsule, center=<-11.1477 0.311 0.5159>, rotation=<rot 90 -2.5294 -2.5277>}
 ]
 thigh_r [{shapeType=Capsule, length=20, radius=7.5, shapeName=thigh_r_capsule, center=<10 0.0002 -1.1014>, rotation=<rot 84.9936 -180 -180>}]
}
constraints [
 {
  bone1Name=Dyn_Belt_FT_02
  bone2Name=Dyn_Belt_FT_01
  swing1LimitDegrees=20
  swing1Motion=Locked
  swing2LimitDegrees=20
  swing2Motion=Locked
  twistLimitDegrees=1
  twistMotion=Locked
 }
 {
  bone1Name=Dyn_Belt_FT_03
  bone2Name=Dyn_Belt_FT_02
  swing1LimitDegrees=20
  swing1Motion=Locked
  swing2LimitDegrees=20
  swing2Motion=Locked
  twistLimitDegrees=1
  twistMotion=Locked
 }
 {bone1Name=Dyn_Belt_FT_04, bone2Name=Dyn_Belt_FT_03, swing1LimitDegrees=20, swing2LimitDegrees=5, twistLimitDegrees=1}
 {bone1Name=Dyn_Belt_FT_05, bone2Name=Dyn_Belt_FT_04, swing1LimitDegrees=20, swing2LimitDegrees=20, twistLimitDegrees=1}
 {bone1Name=Dyn_Belt_FT_06, bone2Name=Dyn_Belt_FT_05, swing1LimitDegrees=20, swing2LimitDegrees=1, twistLimitDegrees=1}
 {bone1Name=Dyn_Belt_BK_02, bone2Name=Dyn_Master_Belt_BK, swing1LimitDegrees=20, swing2LimitDegrees=1, twistLimitDegrees=1}
 {bone1Name=Dyn_Belt_BK_03, bone2Name=Dyn_Belt_BK_02, swing1LimitDegrees=20, swing2LimitDegrees=1, twistLimitDegrees=1}
 {bone1Name=Dyn_Belt_BK_04, bone2Name=Dyn_Belt_BK_03, swing1LimitDegrees=20, swing2LimitDegrees=1, twistLimitDegrees=1}
 {bone1Name=Dyn_Belt_BK_05, bone2Name=Dyn_Belt_BK_04, swing1LimitDegrees=20, swing2LimitDegrees=1, twistLimitDegrees=1}
 {
  bone1Name=thigh_l
  bone2Name=pelvis
  swing1LimitDegrees=45
  swing1Motion=Free
  swing2LimitDegrees=45
  swing2Motion=Free
  twistLimitDegrees=45
  twistMotion=Free
 }
 {
  bone1Name=thigh_r
  bone2Name=pelvis
  swing1LimitDegrees=45
  swing1Motion=Free
  swing2LimitDegrees=45
  swing2Motion=Free
  twistLimitDegrees=45
  twistMotion=Free
 }
 {
  bone1Name=Dyn_Master_Belt_FT
  bone2Name=pelvis
  swing1LimitDegrees=45
  swing1Motion=Free
  swing2LimitDegrees=45
  swing2Motion=Free
  twistLimitDegrees=45
  twistMotion=Free
 }
 {
  bone1Name=Dyn_Master_Belt_BK
  bone2Name=pelvis
  swing1LimitDegrees=45
  swing1Motion=Free
  swing2LimitDegrees=45
  swing2Motion=Free
  twistLimitDegrees=45
  twistMotion=Free
 }
 {bone1Name=Dyn_Belt_FT_06, bone2Name=Dyn_Belt_BK_05, swing1LimitDegrees=20, swing2LimitDegrees=1, twistLimitDegrees=1}
 {
  bone1Name=spine_01
  bone2Name=pelvis
  swing1LimitDegrees=45
  swing1Motion=Free
  swing2LimitDegrees=45
  swing2Motion=Free
  twistLimitDegrees=45
  twistMotion=Free
 }
 {bone1Name=Dyn_Belt_FT_01, bone2Name=Dyn_Master_Belt_FT, swing1LimitDegrees=30, swing2LimitDegrees=10, twistLimitDegrees=1}
]
