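# Post-process style AnimBlueprint for the F_MED_Pink_Emo character body, adding deformation and belt and bracelet secondary motion.
# It copies the attached parent's pose through twist and sculpter Control Rigs, then nudges belt bones by thigh angle.
# RigidBody nodes blend between crouch, in-game and front-end physics assets, followed by fixed belt offsets and bracelet simulation.
# The EventGraph sets those alphas and softens simulation gravity as frame time rises.
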AnimBlueprint /BRCosmetics/Characters/Player/Female/Medium/Bodies/F_MED_Pink_Emo/Meshes/F_MED_Pink_Emo_AnimBP
target_skeleton=&/F_MED_Pink_Emo_Skeleton
graphs [
 AnimGraph {nodes [
   n1 = Get FrontEnd Physics On() -> (FrontEnd Physics On:double)
   n2 = Get Override World Gravity() -> (Override World Gravity:Vector)
   RigidBody!(ComponentPose, Alpha=@n1, OverrideWorldGravity:Vector=@n2, ExternalForce) -> (Pose) {
    bClampLinearTranslationLimitToRefPose=true
    bOverrideWorldGravity=true
    overridePhysicsAsset=&/F_MED_Pink_Emo_Physics_FrontEnd
    simulationSpace=BaseBoneSpace
    alphaBoolBlend {blendOption=Linear}
    alphaScaleBiasClamp {bInterpResult=true}
    baseBoneRef {boneName=pelvis}
    componentAppliedLinearAccClamp=<3000 3000 3000>
    componentLinearAccScale=<0.1 0.1 0.1>
    componentLinearVelScale=<0.1 0.1 0.1>
    overrideWorldGravity {z=-980}
   }
   n3 = Get Override World Gravity()
   n4 = Get FrontEnd Physics Off() -> (FrontEnd Physics Off:double)
   RigidBody(Alpha=@n4, OverrideWorldGravity=@n3) {
    bClampLinearTranslationLimitToRefPose=true
    bOverrideWorldGravity=true
    overridePhysicsAsset=&/F_MED_Pink_Emo_Physics_inGame
    simulationSpace=BaseBoneSpace
    alphaBoolBlend {blendOption=Linear}
    alphaScaleBiasClamp {bInterpResult=true}
    baseBoneRef {boneName=pelvis}
    componentAppliedLinearAccClamp=<3000 3000 3000>
    componentLinearAccScale=<0.1 0.1 0.1>
    componentLinearVelScale=<0.1 0.1 0.1>
    overrideWorldGravity {z=-980}
   }
   n5 = Control Rig(Source=Copy Pose From Mesh() {bUseAttachedParent=true}) {
    alphaBoolBlend {blendOption=Linear}
    controlRigAssetReference {blueprintRigClass=&/F_MED_Pink_Emo_Deform_CtrlRig_twists.F_MED_Pink_Emo_Deform_CtrlRig_twists_C}
   }
   n6 = Control Rig(Source=@n5) {
    alphaBoolBlend {blendOption=Linear}
    controlRigAssetReference {blueprintRigClass=&/F_MED_Pink_Emo_Deform_CtrlRig_sculpters.F_MED_Pink_Emo_Deform_CtrlRig_sculpters_C}
   }
   n7 = Get Thigh L forward() -> (Thigh L forward:double)
   n8 = "Transform (Modify) Bone - Bone: Dyn_Belt_FT_02"(ComponentPose=Local To Component(LocalPose=@n6), Alpha=@n7, Translation="0,0,0", Rotation="0,0,0", Scale="1,1,1") {
    rotationMode=BMM_Additive
    rotationSpace=BCS_BoneSpace
    scaleMode=BMM_Additive
    scaleSpace=BCS_BoneSpace
    translationMode=BMM_Additive
    translationSpace=BCS_BoneSpace
    alphaBoolBlend {blendOption=Linear}
    alphaScaleBiasClamp {bInterpResult=true, interpSpeedDecreasing=1, interpSpeedIncreasing=1}
    boneToModify {boneName=Dyn_Belt_FT_02}
   }
   n9 = "Transform (Modify) Bone - Bone: Dyn_Master_Belt_BK"(ComponentPose=@n8, Alpha=@n7, Translation="0,0,0", Rotation="10,0,0", Scale="1,1,1") {
    rotationMode=BMM_Additive
    rotationSpace=BCS_BoneSpace
    scaleMode=BMM_Additive
    scaleSpace=BCS_BoneSpace
    translationMode=BMM_Additive
    translationSpace=BCS_BoneSpace
    alphaBoolBlend {blendOption=Linear}
    alphaScaleBiasClamp {bInterpResult=true, interpSpeedDecreasing=5, interpSpeedIncreasing=5}
    boneToModify {boneName=Dyn_Master_Belt_BK}
    rotation {pitch=10}
   }
   n10 = "Transform (Modify) Bone - Bone: Dyn_Belt_FT_01"(ComponentPose=@n9, Alpha=Get Thigh R forward() -> (Thigh R forward:double), Translation="0,0,0", Rotation="30,0,0", Scale="1,1,1") {
    rotationMode=BMM_Additive
    rotationSpace=BCS_BoneSpace
    scaleMode=BMM_Additive
    scaleSpace=BCS_BoneSpace
    translationMode=BMM_Additive
    translationSpace=BCS_BoneSpace
    alphaBoolBlend {blendOption=Linear}
    alphaScaleBiasClamp {bInterpResult=true}
    boneToModify {boneName=Dyn_Belt_FT_01}
    rotation {pitch=30}
   }
   n11 = RigidBody(ComponentPose=@n10, Alpha=Get Crouching() -> (Crouching:double), OverrideWorldGravity=@n3) {
    bClampLinearTranslationLimitToRefPose=true
    bOverrideWorldGravity=true
    overridePhysicsAsset=&/F_MED_Pink_Emo_Physics_CROUCH
    simulationSpace=BaseBoneSpace
    alphaBoolBlend {blendOption=Linear}
    alphaScaleBiasClamp {bInterpResult=true}
    baseBoneRef {boneName=pelvis}
    componentAppliedLinearAccClamp=<3000 3000 3000>
    componentLinearAccScale=<0.1 0.1 0.1>
    componentLinearVelScale=<0.1 0.1 0.1>
    overrideWorldGravity {z=-980}
   }
   n12 = RigidBody(ComponentPose=@n11, Alpha=@n4, OverrideWorldGravity=@n3) {
    bClampLinearTranslationLimitToRefPose=true
    bOverrideWorldGravity=true
    overridePhysicsAsset=&F_MED_Pink_Emo_Physics_inGame
    simulationSpace=BaseBoneSpace
    alphaBoolBlend {blendOption=Linear}
    alphaScaleBiasClamp {bInterpResult=true}
    baseBoneRef {boneName=pelvis}
    componentAppliedLinearAccClamp=<3000 3000 3000>
    componentLinearAccScale=<0.1 0.1 0.1>
    componentLinearVelScale=<0.1 0.1 0.1>
    overrideWorldGravity {z=-980}
   }
   n13 = RigidBody(ComponentPose=@n12, Alpha=@n1, OverrideWorldGravity=@n2) {
    bClampLinearTranslationLimitToRefPose=true
    bOverrideWorldGravity=true
    overridePhysicsAsset=&F_MED_Pink_Emo_Physics_FrontEnd
    simulationSpace=BaseBoneSpace
    alphaBoolBlend {blendOption=Linear}
    alphaScaleBiasClamp {bInterpResult=true}
    baseBoneRef {boneName=pelvis}
    componentAppliedLinearAccClamp=<3000 3000 3000>
    componentLinearAccScale=<0.1 0.1 0.1>
    componentLinearVelScale=<0.1 0.1 0.1>
    overrideWorldGravity {z=-980}
   }
   n14 = "Transform (Modify) Bone - Bone: Dyn_Belt_FT_02"!(ComponentPose=@n13, bAlphaBoolEnabled:bool=True, Translation="0.000,0.721,2.548", Rotation="0,0,0", Scale="1,1,1") -> (Pose) {
    alphaInputType=Bool
    rotationMode=BMM_Additive
    translationMode=BMM_Additive
    alphaBoolBlend {blendOption=Linear}
    boneToModify {boneName=Dyn_Belt_FT_02}
    translation {y=0.721, z=2.548}
   }
   n15 = "Transform (Modify) Bone - Bone: Dyn_Belt_FT_05"!(ComponentPose=@n14, bAlphaBoolEnabled:bool=True, Translation="1.608,0.723,0.000", Rotation="0,0,0", Scale="1,1,1") -> (Pose) {
    alphaInputType=Bool
    rotationMode=BMM_Additive
    translationMode=BMM_Additive
    alphaBoolBlend {blendOption=Linear}
    boneToModify {boneName=Dyn_Belt_FT_05}
    translation=<1.608 0.723>
   }
   n16 = RigidBody!(ComponentPose=@n15, Alpha, ExternalForce) -> (Pose) {
    bClampLinearTranslationLimitToRefPose=true
    bOverrideWorldGravity=true
    overridePhysicsAsset=&/F_MED_Pink_Emo_Bracelet_Physics
    simulationSpace=BaseBoneSpace
    alphaBoolBlend {blendOption=Linear}
    baseBoneRef {boneName=lowerarm_r}
    componentAppliedLinearAccClamp=<3000 3000 3000>
    componentLinearAccScale=<0.1 0.1 0.1>
    componentLinearVelScale=<0.1 0.1 0.1>
    overrideWorldGravity {z=-400}
   }
   n17 = "Transform (Modify) Bone - Bone: Dyn_Belt_FT_03"(ComponentPose=@n16, Alpha=1, Translation="0,0,-2", Rotation="0,0,0", Scale="1,1,1") {
    scaleSpace=BCS_BoneSpace
    translationMode=BMM_Additive
    translationSpace=BCS_BoneSpace
    alphaBoolBlend {blendOption=Linear}
    boneToModify {boneName=Dyn_Belt_FT_03}
    translation {z=-2}
   }
   n18 = "Transform (Modify) Bone - Bone: Dyn_Belt_FT_04"(ComponentPose=@n17, Alpha=1, Translation="0,0,1", Rotation="0,0,0", Scale="1,1,1") {
    scaleSpace=BCS_BoneSpace
    translationMode=BMM_Additive
    translationSpace=BCS_BoneSpace
    alphaBoolBlend {blendOption=Linear}
    boneToModify {boneName=Dyn_Belt_FT_04}
    translation {z=1}
   }
   n19 = "Transform (Modify) Bone - Bone: Dyn_Belt_BK_03"(ComponentPose=@n18, Alpha=1, Translation="0,0,-1", Rotation="0,0,0", Scale="1,1,1") {
    rotationMode=BMM_Additive
    scaleMode=BMM_Replace
    scaleSpace=BCS_BoneSpace
    translationMode=BMM_Additive
    translationSpace=BCS_BoneSpace
    alphaBoolBlend {blendOption=Linear}
    boneToModify {boneName=Dyn_Belt_BK_03}
    translation {z=-1}
   }
   Output Pose(Result=Component To Local(ComponentPose=@n19)) {layerGroup=None}
  ]}
 EventGraph {nodes [
   Delay(Duration=0.12)
   > ResetDynamics(InTeleportType=ResetPhysics)
   > ResetDynamics(InTeleportType=TeleportPhysics)
   n1 = Event BlueprintUpdateAnimation()
   > n2 = Sequence!(execute) -> (then_0, then_1, then_2:Exec, then_3:Exec)
   > Set Thigh R forward(Thigh R forward:double=GetAlphaFromReferenceTransformDelta(SkeletalMeshComponent=Get Skel Mesh(), BoneName=thigh_r, BaseName=pelvis, InputRange=(X=-20,Y=-90), TwistAxis="0,0,1")) -> (Output_Get:double)
   Set Thigh L forward(
    execute=@n2.then_1
    Thigh L forward:double=GetAlphaFromReferenceTransformDelta(SkeletalMeshComponent=Get Skel Mesh(), BoneName=thigh_l, BaseName=pelvis, InputRange=(X=20,Y=-100), TwistAxis="0,0,1")
   ) -> (Output_Get:double)
   > Set Override World Gravity(Override World Gravity:Vector="Lerp (Vector)"!(A:Vector="0,0,-800", B:Vector="0,0,-300", Alpha:float=MapRangeClamped(Value=@n1.DeltaTimeX, InRangeA=0.016, InRangeB=0.033)) -> (ReturnValue:Vector)) -> (Output_Get:Vector)
   > Set Crouching(Crouching:double=SelectFloat(A=1, B=0, bPickA=Get bIsCrouching())) -> (Output_Get:double)
   n3 = Get bIsMoving2D()
   n4 = NOT Boolean(A=Get bIsCrouching())
   n5 = Get bIsOnGround() -> (bIsOnGround:bool)
   > Set BeltFloatingFix(BeltFloatingFix:double=SelectFloat(A=SelectFloat(A=0.4, B=1, bPickA=@n3), B=0, bPickA=AND Boolean(A=@n5, B=@n4))) -> (Output_Get:double)
   n6 = Branch(execute=@n2.then_2, Condition=Get bIsInFrontEnd())
   > Set FrontEnd Physics On(FrontEnd Physics On:double=1) -> (Output_Get:double)
   > Set FrontEnd Physics Off(FrontEnd Physics Off:double=0) -> (Output_Get:double)
   Set FrontEnd Physics Off(execute=@n6.else, FrontEnd Physics Off=1)
   > Set FrontEnd Physics On(FrontEnd Physics On=0)
   > n7 = Branch(Condition=Get bIsCrouching())
   > Set FrontEnd Physics Off(FrontEnd Physics Off=0)
   Set FrontEnd Physics Off(execute=@n7.else, FrontEnd Physics Off=1)
   Event BlueprintInitializeAnimation()
   > Set Skel Mesh(Skel Mesh=GetOwnerSkeletalMeshComponent())
   Event OnEndCrouching()
  ]}
]
variables [
 BeltFloatingFix {type=number}
 Crouching {type=number}
 FrontEnd Physics Off {type=number}
 FrontEnd Physics On {type=number}
 NOcrouching {type=number}
 Override World Gravity {type=Vector, default=<0 0 -980>}
 Override World Gravity_0 {type=Vector, default=<0 0 -980>}
 playingEmote {type=boolean}
 Skel Mesh {type=SkeletalMeshComponent}
 Thigh L forward {type=number}
 Thigh R forward {type=number}
]
