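# Physics asset for the Pink Emo medium female character's FaceAcc mesh, set up for dynamic hair motion with capsule bodies on the hair, head, neck and clavicles.
# Back hair strands and three main clumps mostly constrain directly to Dyn_Master_Hair, with varied swing limits and very tight twist.
# Front hair branches through Dyn_Master_Front_Hair and carries three further strands.
# Master hair, head and neck links have free swing and twist, while clavicle bodies are unconstrained.
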PhysicsAsset /BRCosmetics/Characters/Player/Female/Medium/Bodies/F_MED_Pink_Emo/Meshes/Parts/F_MED_Pink_Emo_FaceAcc_Physics
bodies {
 Dyn_Back_Hair1 [{shapeType=Capsule, length=2, radius=1, shapeName=Dyn_Back_Hair1_capsule, center=<1 0 0>, rotation=<rot 90 0 0>}]
 Dyn_Back_Hair13 [{shapeType=Capsule, length=5, radius=1, shapeName=Dyn_Back_Hair13_capsule, center=<2.8957 0 0.0001>, rotation=<rot 90 0 0>}]
 Dyn_Back_Hair14 [{shapeType=Capsule, length=2, radius=1, shapeName=Dyn_Back_Hair14_capsule, center=<1 0 0>, rotation=<rot 90 0 0>}]
 Dyn_Back_Hair15 [{shapeType=Capsule, length=2, radius=0.75, shapeName=Dyn_Back_Hair15_capsule, center=<1 -0.3711 0>, rotation=<rot 90 179.9994 179.9994>}]
 Dyn_Back_Hair16 [{shapeType=Capsule, length=2, radius=1, shapeName=Dyn_Back_Hair16_capsule, center=<1 0 0>, rotation=<rot 90 0 0>}]
 Dyn_Back_Hair17 [{shapeType=Capsule, length=2, radius=1, shapeName=Dyn_Back_Hair17_capsule, center=<1 0 0>, rotation=<rot 90 0 0>}]
 Dyn_Back_Hair18 [{shapeType=Capsule, length=2, radius=1, shapeName=Dyn_Back_Hair18_capsule, center=<1 0 0>, rotation=<rot 90 0 0>}]
 Dyn_Back_Hair19 [{shapeType=Capsule, length=2, radius=1, shapeName=Dyn_Back_Hair19_capsule, center=<1 0 0>, rotation=<rot 90 0 0>}]
 Dyn_Back_Hair2 [{shapeType=Capsule, length=1.75, radius=0.5, shapeName=Dyn_Back_Hair2_capsule, center=<1 0 0>, rotation=<rot 90 0 0>}]
 Dyn_Back_Hair20 [{shapeType=Capsule, length=2.5, radius=0.5, shapeName=Dyn_Back_Hair20_capsule, center=<2.2891 -0.8126 0>, rotation=<rot 90 0 0>}]
 Dyn_Back_Hair21 [{shapeType=Capsule, length=2, radius=1, shapeName=Dyn_Back_Hair21_capsule, center=<1 -0.1218 0>, rotation=<rot 90 0 0>}]
 Dyn_Back_Hair22 [{shapeType=Capsule, length=2, radius=1, shapeName=Dyn_Back_Hair22_capsule, center=<1 0 0>, rotation=<rot 90 0 0>}]
 Dyn_Back_Hair23 [{shapeType=Capsule, length=2, radius=1, shapeName=Dyn_Back_Hair23_capsule, center=<1 0 0>, rotation=<rot 90 0 0>}]
 Dyn_Back_Hair24 [{shapeType=Capsule, length=2, radius=1, shapeName=Dyn_Back_Hair24_capsule, center=<1 0 0>, rotation=<rot 90 0 0>}]
 Dyn_Back_Hair26 [{shapeType=Capsule, length=2, radius=1, shapeName=Dyn_Back_Hair26_capsule, center=<1 -0.1256 0>, rotation=<rot 90 0 0>}]
 Dyn_Back_Hair3 [{shapeType=Capsule, length=2, radius=1, shapeName=Dyn_Back_Hair3_capsule, center=<1 0 0>, rotation=<rot 90 0 0>}]
 Dyn_Back_Hair4 [{shapeType=Capsule, length=1.75, radius=0.5, shapeName=Dyn_Back_Hair4_capsule, center=<1.9866 -0.949 -0.2547>, rotation=<rot 90 0 0>}]
 Dyn_Back_Hair5 [{shapeType=Capsule, length=2, radius=1, shapeName=Dyn_Back_Hair5_capsule, center=<2.3614 -0.5787 0.7148>, rotation=<rot 90 0 -10>}]
 Dyn_Back_Hair6 [{shapeType=Capsule, length=2, radius=1, shapeName=Dyn_Back_Hair6_capsule, center=<1 0 0>, rotation=<rot 90 0 0>}]
 Dyn_Back_Hair7 [{shapeType=Capsule, length=2, radius=1, shapeName=Dyn_Back_Hair7_capsule, center=<1 0 0>, rotation=<rot 90 0 0>}]
 Dyn_Back_Hair_Main_c [{
   shapeType=Capsule
   length=6.1616
   radius=2.7169
   shapeName=Dyn_Back_Hair_Main_c_capsule
   center=<5.9529 -0.8826 0.0636>
   rotation=<rot 1.6223 79.7043 -83.9324>
  }]
 Dyn_Back_Hair_Main_l [{
   shapeType=Capsule
   length=3.856
   radius=3.0489
   shapeName=Dyn_Back_Hair_Main_l_capsule
   center=<6.9904 0.3682 0.9402>
   rotation=<rot 2.669 -104.2507 -84.2626>
  }]
 Dyn_Back_Hair_Main_r [{
   shapeType=Capsule
   length=4.2107
   radius=3
   shapeName=Dyn_Back_Hair_Main_r_capsule
   center=<6.0118 0.5743 0.9924>
   rotation=<rot 1.0644 -100.1454 -87.7165>
  }]
 Dyn_Front_Hair [{shapeType=Capsule, length=2, radius=3, shapeName=Dyn_Front_Hair_capsule, center=<1 -0.2731 0>, rotation=<rot 90 0 0>}]
 Dyn_Master_Front_Hair [{shapeType=Capsule, length=0.5, radius=0.5, shapeName=Dyn_Master_Front_Hair_capsule}]
 Dyn_Master_Hair [{shapeType=Capsule, length=0.5, radius=0.5, shapeName=Dyn_Master_Hair_capsule}]
 clavicle_l [{shapeType=Capsule, length=12, radius=5, shapeName=clavicle_l_capsule, center=<2.823 -2.68 0>, rotation=<rot 90 0 0>}]
 clavicle_r [{shapeType=Capsule, length=12, radius=5, shapeName=clavicle_r_capsule, center=<-2.577 1.9743 0>, rotation=<rot 90 0 0>}]
 head [
  {shapeType=Capsule, length=2.25, radius=7, shapeName=head_capsule, center=<0 0.4707 0>, rotation=<rot 90 0 0>}
  {shapeType=Capsule, length=8, radius=7, shapeName=head_capsule, center=<0 1.721 0>, rotation=<rot 90 0 0>}
 ]
 neck_01 [
  {shapeType=Capsule, length=8, radius=4.5, shapeName=neck_01_capsule, rotation=<rot 90 -14.0363 -19.0363>}
  {shapeType=Capsule, length=5, radius=4, shapeName=neck_01_capsule, center=<-1.7582 0 0>}
 ]
}
constraints [
 {bone1Name=Dyn_Back_Hair1, bone2Name=Dyn_Master_Hair, swing1LimitDegrees=20, swing2LimitDegrees=1, twistLimitDegrees=1}
 {bone1Name=Dyn_Back_Hair2, bone2Name=Dyn_Master_Hair, swing1LimitDegrees=20, swing2LimitDegrees=20, twistLimitDegrees=1}
 {bone1Name=Dyn_Back_Hair3, bone2Name=Dyn_Master_Hair, swing1LimitDegrees=20, swing2LimitDegrees=1, twistLimitDegrees=1}
 {bone1Name=Dyn_Back_Hair4, bone2Name=Dyn_Master_Hair, swing1LimitDegrees=20, swing2LimitDegrees=1, twistLimitDegrees=1}
 {bone1Name=Dyn_Back_Hair5, bone2Name=Dyn_Master_Hair, swing1LimitDegrees=20, swing2LimitDegrees=50, twistLimitDegrees=1}
 {bone1Name=Dyn_Back_Hair6, bone2Name=Dyn_Master_Hair, swing1LimitDegrees=20, swing2LimitDegrees=1, twistLimitDegrees=1}
 {bone1Name=Dyn_Back_Hair7, bone2Name=Dyn_Master_Hair, swing1LimitDegrees=20, swing2LimitDegrees=1, twistLimitDegrees=1}
 {bone1Name=Dyn_Back_Hair13, bone2Name=Dyn_Master_Hair, swing1LimitDegrees=20, swing2LimitDegrees=20, twistLimitDegrees=5}
 {bone1Name=Dyn_Back_Hair14, bone2Name=Dyn_Master_Hair, swing1LimitDegrees=20, swing2LimitDegrees=20, twistLimitDegrees=1}
 {bone1Name=Dyn_Back_Hair15, bone2Name=Dyn_Master_Hair, swing1LimitDegrees=20, swing2LimitDegrees=30, twistLimitDegrees=1}
 {bone1Name=Dyn_Back_Hair16, bone2Name=Dyn_Master_Hair, swing1LimitDegrees=20, swing2LimitDegrees=1, twistLimitDegrees=1}
 {bone1Name=Dyn_Back_Hair17, bone2Name=Dyn_Master_Hair, swing1LimitDegrees=20, swing2LimitDegrees=1, twistLimitDegrees=1}
 {bone1Name=Dyn_Back_Hair18, bone2Name=Dyn_Master_Hair, swing1LimitDegrees=20, swing2LimitDegrees=1, twistLimitDegrees=1}
 {bone1Name=Dyn_Back_Hair19, bone2Name=Dyn_Master_Hair, swing1LimitDegrees=20, swing2LimitDegrees=1, twistLimitDegrees=1}
 {bone1Name=Dyn_Back_Hair22, bone2Name=Dyn_Master_Hair, swing1LimitDegrees=20, swing2LimitDegrees=1, twistLimitDegrees=1}
 {bone1Name=Dyn_Back_Hair23, bone2Name=Dyn_Master_Hair, swing1LimitDegrees=20, swing2LimitDegrees=1, twistLimitDegrees=1}
 {bone1Name=Dyn_Back_Hair24, bone2Name=Dyn_Master_Hair, swing1LimitDegrees=20, swing2LimitDegrees=1, twistLimitDegrees=1}
 {
  bone1Name=Dyn_Master_Hair
  bone2Name=head
  swing1LimitDegrees=45
  swing1Motion=Free
  swing2LimitDegrees=45
  swing2Motion=Free
  twistLimitDegrees=45
  twistMotion=Free
 }
 {bone1Name=Dyn_Front_Hair, bone2Name=Dyn_Master_Front_Hair, swing1LimitDegrees=30, swing2LimitDegrees=1, twistLimitDegrees=1}
 {
  bone1Name=Dyn_Master_Front_Hair
  bone2Name=Dyn_Master_Hair
  swing1LimitDegrees=45
  swing1Motion=Free
  swing2LimitDegrees=45
  swing2Motion=Free
  twistLimitDegrees=45
  twistMotion=Free
 }
 {bone1Name=Dyn_Back_Hair26, bone2Name=Dyn_Front_Hair, swing1LimitDegrees=30, swing2LimitDegrees=10, twistLimitDegrees=5}
 {bone1Name=Dyn_Back_Hair21, bone2Name=Dyn_Front_Hair, swing1LimitDegrees=20, swing2LimitDegrees=20, twistLimitDegrees=5}
 {bone1Name=Dyn_Back_Hair20, bone2Name=Dyn_Front_Hair, swing1LimitDegrees=40, swing2LimitDegrees=40, twistLimitDegrees=5}
 {
  bone1Name=head
  bone2Name=neck_01
  swing1LimitDegrees=45
  swing1Motion=Free
  swing2LimitDegrees=45
  swing2Motion=Free
  twistLimitDegrees=45
  twistMotion=Free
 }
 {bone1Name=Dyn_Back_Hair_Main_l, bone2Name=Dyn_Master_Hair, swing1LimitDegrees=30, swing2LimitDegrees=20, twistLimitDegrees=1}
 {bone1Name=Dyn_Back_Hair_Main_c, bone2Name=Dyn_Master_Hair, swing1LimitDegrees=30, swing2LimitDegrees=1, twistLimitDegrees=1}
 {bone1Name=Dyn_Back_Hair_Main_r, bone2Name=Dyn_Master_Hair, swing1LimitDegrees=30, swing2LimitDegrees=5, twistLimitDegrees=5}
]
