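# Animation Blueprint for the F_MED_Pink_Emo face accessory mesh, adding physics motion to a part that follows the character's body.
# It copies the attached parent's pose, then runs a RigidBody node with an override physics asset, simulated relative to spine_05.
# The simulation uses overridden gravity and scaled, clamped component motion.
# Three additive Transform Bone nodes for back-hair bones sit unconnected, and the event graph holds no logic.
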AnimBlueprint /BRCosmetics/Characters/Player/Female/Medium/Bodies/F_MED_Pink_Emo/Meshes/Parts/F_MED_Pink_Emo_FaceAcc_AnimBP
target_skeleton=&/F_MED_Pink_Emo_FaceAcc_Skeleton
graphs [
 AnimGraph {nodes [
   "Transform (Modify) Bone - Bone: Dyn_Master_Back_Hair_r"(Alpha=1, Translation="0,0,0", Rotation="0,-60,0", Scale="0.5,0.5,0.5") {
    rotationMode=BMM_Additive
    rotationSpace=BCS_BoneSpace
    scaleMode=BMM_Additive
    scaleSpace=BCS_BoneSpace
    translationMode=BMM_Additive
    translationSpace=BCS_BoneSpace
    alphaBoolBlend {blendOption=Linear}
    boneToModify {boneName=Dyn_Master_Back_Hair_r}
    rotation {yaw=-60}
    scale=<0.5 0.5 0.5>
   }
   "Transform (Modify) Bone - Bone: Dyn_Master_Back_Hair"(Alpha=1, Translation="0,-3,0", Rotation="0,-60,0", Scale="0.8,0.8,0.8") {
    rotationMode=BMM_Additive
    rotationSpace=BCS_BoneSpace
    scaleMode=BMM_Additive
    scaleSpace=BCS_BoneSpace
    translationMode=BMM_Additive
    translationSpace=BCS_BoneSpace
    alphaBoolBlend {blendOption=Linear}
    boneToModify {boneName=Dyn_Master_Back_Hair}
    rotation {yaw=-60}
    scale=<0.8 0.8 0.8>
    translation {y=-3}
   }
   "Transform (Modify) Bone - Bone: Dyn_Master_Back_Hair_c"(Alpha=1, Translation="0,-5,0", Rotation="0,-20,0", Scale="1,1,1") {
    rotationMode=BMM_Additive
    rotationSpace=BCS_BoneSpace
    scaleMode=BMM_Additive
    scaleSpace=BCS_BoneSpace
    translationMode=BMM_Additive
    translationSpace=BCS_BoneSpace
    alphaBoolBlend {blendOption=Linear}
    boneToModify {boneName=Dyn_Master_Back_Hair_c}
    rotation {yaw=-20}
    translation {y=-5}
   }
   n1 = RigidBody(ComponentPose=Local To Component(LocalPose=Copy Pose From Mesh() {bUseAttachedParent=true})) {
    bClampLinearTranslationLimitToRefPose=true
    bFreezeIncomingPoseOnStart=true
    bOverrideWorldGravity=true
    overridePhysicsAsset=&/F_MED_Pink_Emo_FaceAcc_Physics
    simulationSpace=BaseBoneSpace
    alphaBoolBlend {blendOption=Linear}
    baseBoneRef {boneName=spine_05}
    componentAppliedLinearAccClamp=<3000 3000 3000>
    componentLinearAccScale=<0.1 0.3 0.1>
    componentLinearVelScale=<0.1 0.3 0.1>
    overrideWorldGravity {z=-980}
   }
   Output Pose(Result=Component To Local(ComponentPose=@n1)) {layerGroup=None}
  ]}
 EventGraph {nodes [Event BlueprintUpdateAnimation()]}
]
variables []
